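# Physics asset for the Floret Love gadget prop mesh in the cosmetics animation content.
# It defines a single body on the "main" bone, shaped as one short, wide capsule that is offset and rotated relative to that bone.
# It has no constraints, so the gadget collides as one rigid piece with no jointed parts.
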PhysicsAsset /BRCosmetics/Animation/Props/Gadgets_Floret_Love/Meshes/Floret_Love_Gadget_Physics
bodies {main [{shapeType=Capsule, length=19.0564, radius=18.6784, center=<2.2326 1.5382 5.2714>, rotation=<rot 0 131.6114 -84.0023>}]}
constraints []
